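# Niagara effect for the HighTower Tomato casual skin's emote fade-in, drawing short-lived cyan-to-blue "digital bit" sprites on the character's Armor mesh region.
# Spawn rate tapers to zero across the emitter loop while particles drift inward against the surface normal and shrink.
# A local SphereMaskEdge module grows its radii over the loop, culling particles beyond the outer edge to create an expanding reveal.
# Two texture samples multiply into particle alpha.
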NiagaraSystem /BRCosmetics/Characters/Player/Male/Medium/Bodies/M_MED_HighTower_Tomato_Casual/FX/NS_HighTower_Tomato_Emote_FadeIn
dependencies {
 usedDataInterfaces [&NiagaraDataInterfaceCurve, &NiagaraDataInterfaceSkeletalMesh]
 usedDynamicInputs []
 usedModules [&SystemState]
}
emitters {Digital_Bits {
  emitterUpdateScript [
   &EmitterState(Life Cycle Mode=::NewEnumerator1, Loop Behavior=::NewEnumerator1, Loop Duration=2.27, UseLoopDelay=false)
   &SpawnRate(
    SpawnRate=&FloatFromCurve(
     FloatCurve=DataInterface() {propertyValues {Curve {keys [<at 0 1 RCIM_Cubic>, {arriveTangent=-2.7657, interpMode=RCIM_Cubic, leaveTangent=-2.7657, tangentMode=RCTM_User, time=1, value=0}]}}}
     CurveIndex=$Emitter.NormalizedLoopAge
     Scale Curve=1000
    )
    Use Spawn Probability=false
   )
  ]
  particleSpawnScript [
   /Niagara/Modules/Spawn/Initialization/InitializeParticle(
    Color=&MakeLinearColorFromVectorAndFloat(
     "Vector (RGB)"=&MakeVectorFromLinearColorRGB(Color=&UniformRangedLinearColor(Minimum=<rgba 0 0.6667 1 1>, Maximum=<rgba 0 3.3333 5 1>))
     "Float (Alpha)"=&UniformRangedFloat(Minimum=0.25, Maximum=1)
    )
    Lifetime=&UniformRangedFloat(Minimum=0.05, Maximum=0.1)
    Module.Write Color=true
    Module.Write Lifetime=true
    Module.Write Mass=true
    Module.Write Material Random=false
    Module.Write Position=false
    Module.Write Scale=false
    Module.Write SpriteRotation=false
    Module.Write SpriteSize=true
    Position=&SimulationPosition()
    Sprite Size=<1 1>
   )
   &SkeletalMeshLocation(
    Inherit Velocity=false
    MaskSpawnGroup=false
    Mesh Sampling Type=::NewEnumerator7
    Module.Override Randomness=false
    OffsetAlongNormal=false
    Skeletal Mesh=DataInterface() {propertyValues {
      DefaultMesh=null
      ExcludeBoneName=root
      OptionalComponentClass=null
      PreviewMesh=/BRCosmetics/Characters/Player/Male/Medium/Bodies/M_MED_HighTower_Tomato_Casual/Meshes/M_MED_HighTower_Tomato_Casual
      SkinningMode=PreSkin
      SoftSourceActor=null
      SourceMode=Default
      UvSetIndex=0
      WholeMeshLOD=-1
      bExcludeBone=false
      bReadDeformedGeometry=true
      bRequireCurrentFrameData=true
      ComponentTags []
      FilteredBones []
      FilteredSockets []
      MeshUserParameter {parameter {name=None, typeDefHandle {registeredTypeIndex=1}, varData []}}
      SamplingRegions [Armor]
     }}
    Triangle Sampling Mode:ENiagaraTriangleSamplingMode=::NewEnumerator5
   )
   /SetVariables(
    Particles.NormalVelocity:Vector3f=$Output.SkeletalMeshLocation.SampledNormal
    Particles.TextureSampleA:LinearColor=$Output.SampleTexture.SampledColor
    Particles.TextureSampleB:LinearColor=$Output.SampleTexture001.SampledColor
   ) {isSetParameters=true}
   &CameraOffset(Camera Offset Amount=50)
  ]
  particleUpdateScript [
   &ParticleState()
   &self:SphereMaskEdge(
    Cull Particles outside of max distance=true
    Max Distance:NiagaraFloat=&FloatFromCurve(
     FloatCurve=DataInterface() {propertyValues {Curve {keys [
         {arriveTangent=1.1063, interpMode=RCIM_Cubic, leaveTangent=1.1063, tangentMode=RCTM_User, time=0.001, value=0}
         {arriveTangent=1.1925, interpMode=RCIM_Cubic, leaveTangent=1.1925, tangentMode=RCTM_User, time=0.9, value=1}
        ]}}}
     CurveIndex=$Emitter.NormalizedLoopAge
     Scale Curve=128
    )
    Min Distance:NiagaraFloat=&FloatFromCurve(
     FloatCurve=DataInterface() {propertyValues {Curve {keys [
         {arriveTangent=0.7543, interpMode=RCIM_Cubic, leaveTangent=0.7543, tangentMode=RCTM_User, time=0.004, value=0}
         {arriveTangent=1.004, interpMode=RCIM_Cubic, leaveTangent=1.004, time=0.6, value=0.451}
         {arriveTangent=1.4135, interpMode=RCIM_Cubic, leaveTangent=1.4135, tangentMode=RCTM_User, time=1, value=1}
        ]}}}
     CurveIndex=$Emitter.NormalizedLoopAge
     Scale Curve=128
    )
    SphereCenter:Vector3f=<0 8.5591 128>
   ) {module=SphereMaskEdge}
   &ScaleColor(
    Scale Alpha=&Multiply_Float(
     A=&MakeFloatFromLinearColor(LinearColor=$Particles.TextureSampleA, Channel:ENiagaraLinearColor_Channels=::NewEnumerator0)
     B=&MakeFloatFromLinearColor(LinearColor=$Particles.TextureSampleB, Channel:ENiagaraLinearColor_Channels=::NewEnumerator0)
    )
    ScaleA=true
    ScaleRGB=true
   )
   &AddVelocity(Velocity=&Multiply_Vector(A=$Particles.NormalVelocity, B=&VectorFromFloat(Value=-1)))
   &SolveForcesAndVelocity(Clamp Velocity=false, Limit Acceleration=false)
   &ScaleColor(ScaleA=true, ScaleRGB=true) {module=ScaleColor001}
   &ScaleSpriteSize(Scale Factor=&Vector2DFromFloat(Value=&FloatFromCurve(FloatCurve=DataInterface() {propertyValues {Curve {keys [<at 0.5 1>,<at 1 0>]}}})))
  ]
  rendererClasses [&NiagaraSpriteRendererProperties]
  renderers [{rendererClass=&NiagaraSpriteRendererProperties}]
 }}
summary {
 systemName=NS_HighTower_Tomato_Emote_FadeIn
 emitters [{emitterName=Digital_Bits, rendererClasses [&NiagaraSpriteRendererProperties]}]
 userVariables []
}
