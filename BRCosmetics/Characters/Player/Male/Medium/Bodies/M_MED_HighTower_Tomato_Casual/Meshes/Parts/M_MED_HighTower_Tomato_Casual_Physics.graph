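# Physics asset for the hair of the HighTower Tomato Casual medium male player cosmetic, covering left and right bangs and back hair.
# Small boxes on the dyn_ bones form short constraint hierarchies from three base bodies, with the left bangs forking at their parent and three back pieces sharing one base.
# Swing limits are looser on bangs than back hair, twist is tight throughout, and the head is an unconstrained capsule.
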PhysicsAsset /BRCosmetics/Characters/Player/Male/Medium/Bodies/M_MED_HighTower_Tomato_Casual/Meshes/Parts/M_MED_HighTower_Tomato_Casual_Physics
bodies {
 dyn_bangs_l_01 [{shapeType=Box, extentX=1.8619, extentY=1.01, extentZ=0.5062, center=<0.9285 0 -0.0001>}]
 dyn_bangs_l_02 [{shapeType=Box, extentX=1.4757, extentY=1.0887, extentZ=0.6457, center=<0.7128 0 0>}]
 dyn_bangs_l_base [{shapeType=Box, extentX=0.5, extentY=0.5, extentZ=0.5}]
 dyn_bangs_l_parent [{shapeType=Box, extentX=2.1826, extentY=2.6345, extentZ=0.7119, center=<1.1389 0 0>}]
 dyn_bangs_r_01 [{shapeType=Box, extentX=1.8527, extentY=1.7986, extentZ=0.697, center=<0.8559 0 0>}]
 dyn_bangs_r_base [{shapeType=Box, extentX=0.5, extentY=0.5, extentZ=0.5}]
 dyn_bangs_r_parent [{shapeType=Box, extentX=2.2072, extentY=2.4459, extentZ=0.7529, center=<1.0311 0 0>}]
 dyn_hair_back_01 [{shapeType=Box, extentX=1.6133, extentY=0.8419, extentZ=0.5015, center=<0.7841 0 0>}]
 dyn_hair_back_02 [{shapeType=Box, extentX=1.2623, extentY=0.971, extentZ=0.5015, center=<0.6097 0 -0.0001>}]
 dyn_hair_back_03 [{shapeType=Box, extentX=1.6133, extentY=0.9961, extentZ=0.5015, center=<0.6491 0 0>}]
 dyn_hair_base [{shapeType=Box, extentX=0.5, extentY=0.5, extentZ=0.5}]
 head [{shapeType=Capsule, length=3.1399, radius=3.3911, center=<6.4022 0.8735 0>, rotation=<rot 90 0 0>}]
}
constraints [
 {bone1Name=dyn_bangs_r_parent, bone2Name=dyn_bangs_r_base, swing1LimitDegrees=20, swing2LimitDegrees=20, twistLimitDegrees=2}
 {bone1Name=dyn_bangs_r_01, bone2Name=dyn_bangs_r_parent, swing1LimitDegrees=20, swing2LimitDegrees=20, twistLimitDegrees=2}
 {bone1Name=dyn_bangs_l_parent, bone2Name=dyn_bangs_l_base, swing1LimitDegrees=20, swing2LimitDegrees=20, twistLimitDegrees=2}
 {bone1Name=dyn_bangs_l_01, bone2Name=dyn_bangs_l_parent, swing1LimitDegrees=20, swing2LimitDegrees=20, twistLimitDegrees=2}
 {bone1Name=dyn_bangs_l_02, bone2Name=dyn_bangs_l_parent, swing1LimitDegrees=20, swing2LimitDegrees=20, twistLimitDegrees=2}
 {bone1Name=dyn_hair_back_01, bone2Name=dyn_hair_base, swing1LimitDegrees=10, swing2LimitDegrees=10, twistLimitDegrees=2}
 {bone1Name=dyn_hair_back_02, bone2Name=dyn_hair_base, swing1LimitDegrees=10, swing2LimitDegrees=10, twistLimitDegrees=2}
 {bone1Name=dyn_hair_back_03, bone2Name=dyn_hair_base, swing1LimitDegrees=10, swing2LimitDegrees=10, twistLimitDegrees=2}
]
